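# Physics asset for the F_MED_Candy_Summer female body cosmetic, covering two dangling belt pouches and two belt rounds.
# Each accessory has a box or capsule body joined to a small capsule base bone by narrow swing and twist limits.
# The base bones and right thigh connect to the pelvis with free angular motion.
# Pelvis and right thigh capsules are the only body shapes besides the accessories.
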PhysicsAsset /BRCosmetics/Characters/Player/Female/Medium/Bodies/F_MED_Candy_Summer/Meshes/F_MED_Candy_Summer_Physics
bodies {
 dyn_belt_round_a [{shapeType=Capsule, length=2, radius=1.5, center=<2 0 0>, rotation=<rot 90 0 0>}]
 dyn_belt_round_a_base [{shapeType=Capsule, length=0.505, radius=0.505, rotation=<rot 0 -90 0>}]
 dyn_belt_round_b [{shapeType=Capsule, length=2, radius=1.5, center=<2 0 0>, rotation=<rot 90 0 0>}]
 dyn_belt_round_b_base [{shapeType=Capsule, length=0.505, radius=0.505, rotation=<rot 0 -90 0>}]
 dyn_pouch_a [{shapeType=Box, extentX=9, extentY=7, extentZ=2.5, center=<2.3292 -0.2472 -2.2806>}]
 dyn_pouch_a_base [{shapeType=Capsule, length=0.505, radius=0.505, rotation=<rot 0 -90 0>}]
 dyn_pouch_b [{shapeType=Box, extentX=8.5, extentY=6.5, extentZ=3, center=<3.0784 -0.3455 -2.31>, rotation=<rot 0 0 -3.2461>}]
 dyn_pouch_b_base [{shapeType=Capsule, length=0.505, radius=0.505, rotation=<rot 0 -90 0>}]
 pelvis [{shapeType=Capsule, length=9.75, radius=10, center=<0 -1.5133 0>}]
 thigh_r [{shapeType=Capsule, length=23.25, radius=7.5, center=<16.0599 0.0003 0>, rotation=<rot 87.4999 -180 -180>}]
}
constraints [
 {
  bone1Name=thigh_r
  bone2Name=pelvis
  swing1LimitDegrees=45
  swing1Motion=Free
  swing2LimitDegrees=45
  swing2Motion=Free
  twistLimitDegrees=45
  twistMotion=Free
 }
 {bone1Name=dyn_pouch_a, bone2Name=dyn_pouch_a_base, swing1LimitDegrees=30, swing2LimitDegrees=30, twistLimitDegrees=15}
 {bone1Name=dyn_pouch_b, bone2Name=dyn_pouch_b_base, swing1LimitDegrees=15, swing2LimitDegrees=25, twistLimitDegrees=10}
 {
  bone1Name=dyn_pouch_a_base
  bone2Name=pelvis
  swing1LimitDegrees=45
  swing1Motion=Free
  swing2LimitDegrees=45
  swing2Motion=Free
  twistLimitDegrees=45
  twistMotion=Free
 }
 {
  bone1Name=dyn_pouch_b_base
  bone2Name=pelvis
  swing1LimitDegrees=45
  swing1Motion=Free
  swing2LimitDegrees=45
  swing2Motion=Free
  twistLimitDegrees=45
  twistMotion=Free
 }
 {
  bone1Name=dyn_belt_round_a_base
  bone2Name=pelvis
  swing1LimitDegrees=45
  swing1Motion=Free
  swing2LimitDegrees=45
  swing2Motion=Free
  twistLimitDegrees=45
  twistMotion=Free
 }
 {
  bone1Name=dyn_belt_round_b_base
  bone2Name=pelvis
  swing1LimitDegrees=45
  swing1Motion=Free
  swing2LimitDegrees=45
  swing2Motion=Free
  twistLimitDegrees=45
  twistMotion=Free
 }
 {bone1Name=dyn_belt_round_a, bone2Name=dyn_belt_round_a_base, swing1LimitDegrees=10, swing2LimitDegrees=15, twistLimitDegrees=15}
 {bone1Name=dyn_belt_round_b, bone2Name=dyn_belt_round_b_base, swing1LimitDegrees=10, swing2LimitDegrees=15, twistLimitDegrees=15}
]
