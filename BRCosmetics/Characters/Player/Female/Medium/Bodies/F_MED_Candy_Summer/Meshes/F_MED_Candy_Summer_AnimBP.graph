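# Animation Blueprint for the F_MED_Candy_Summer body, adding secondary motion to thighs, belt and pouches over a pose copied from the attached parent mesh.
# Constraints and additive bone offsets, scaled by thigh rotations measured against the pelvis each update, correct thigh volume and move accessories.
# Spring-driven AnimDynamics on belt bones and a pelvis-space RigidBody node follow.
# Crouch, jump and fall states set the main belt's dynamics alpha, disabling it when crouched.
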AnimBlueprint /BRCosmetics/Characters/Player/Female/Medium/Bodies/F_MED_Candy_Summer/Meshes/F_MED_Candy_Summer_AnimBP
target_skeleton=&/F_MED_Candy_Summer_Skeleton
graphs [
 AnimGraph {nodes [
   n1 = "Constraint - thigh_volume_l (R)"(ComponentPose=Local To Component(LocalPose=Copy Pose From Mesh() {bUseAttachedParent=true}), Alpha=0.5, ConstraintWeights_0=1) {
    alphaBoolBlend {blendOption=Linear}
    boneToModify {boneName=thigh_volume_l}
    constraintSetup [{transformType=Rotation, targetBone {boneName=thigh_l}}]
    constraintWeights [1]
   }
   n2 = "Constraint - thigh_volume_r (R)"(ComponentPose=@n1, Alpha=0.5, ConstraintWeights_0=1) {
    alphaBoolBlend {blendOption=Linear}
    boneToModify {boneName=thigh_volume_r}
    constraintSetup [{transformType=Rotation, targetBone {boneName=thigh_r}}]
    constraintWeights [1]
   }
   n3 = "Transform (Modify) Bone - Bone: thigh_volume_l"(ComponentPose=@n2, Alpha=Get thigh_rot_l(), Translation="-1,-2,1", Rotation="0,0,0", Scale="1,1,1") {
    translationMode=BMM_Additive
    translationSpace=BCS_BoneSpace
    alphaBoolBlend {blendOption=Linear}
    alphaScaleBiasClamp {bMapRange=true, inRange {max=0, min=-100}, outRange {max=0, min=1}}
    boneToModify {boneName=thigh_volume_l}
    translation=<-1 -2 1>
   }
   n4 = "Transform (Modify) Bone - Bone: thigh_volume_r"(ComponentPose=@n3, Alpha=Get thigh_rot_r(), Translation="1,2,-1", Rotation="0,0,0", Scale="1,1,1") {
    translationMode=BMM_Additive
    translationSpace=BCS_BoneSpace
    alphaBoolBlend {blendOption=Linear}
    alphaScaleBiasClamp {bMapRange=true, inRange {max=0, min=-50}, outRange {max=0, min=1}}
    boneToModify {boneName=thigh_volume_r}
    translation=<1 2 -1>
   }
   n5 = "Constraint - dyn_belt_round_a_base (PP)"!(ComponentPose=@n4, Alpha=1, ConstraintWeights_0=0.1589, ConstraintWeights_1:float=0.8411) -> (Pose) {
    alphaBoolBlend {blendOption=Linear}
    boneToModify {boneName=dyn_belt_round_a_base}
    constraintSetup [{targetBone {boneName=spine_01}}, {targetBone {boneName=pelvis}}]
    constraintWeights [1,1]
   }
   n6 = "Constraint - dyn_belt_round_b_base (PPP)"!(ComponentPose=@n5, Alpha=1, ConstraintWeights_0=0.0087, ConstraintWeights_1:float=0.3401, ConstraintWeights_2:float=0.6512) -> (Pose) {
    alphaBoolBlend {blendOption=Linear}
    boneToModify {boneName=dyn_belt_round_b_base}
    constraintSetup [{targetBone {boneName=dyn_belt_main}}, {targetBone {boneName=spine_01}}, {targetBone {boneName=pelvis}}]
    constraintWeights [1,1,1]
   }
   n7 = "Constraint - dyn_pouch_a_base (P)"!(ComponentPose=@n6, Alpha=1, ConstraintWeights_0=1, ConstraintWeights_1:float=0) -> (Pose) {
    alphaBoolBlend {blendOption=Linear}
    boneToModify {boneName=dyn_pouch_a_base}
    constraintSetup [{targetBone {boneName=dyn_belt_attach}}]
    constraintWeights [1,0]
   }
   n8 = "Constraint - dyn_pouch_b_base (PP)"!(ComponentPose=@n7, Alpha=1, ConstraintWeights_0=0.0514, ConstraintWeights_1:float=0.9486) -> (Pose) {
    alphaBoolBlend {blendOption=Linear}
    boneToModify {boneName=dyn_pouch_b_base}
    constraintSetup [{targetBone {boneName=spine_01}}, {targetBone {boneName=dyn_belt_attach}}]
    constraintWeights [1,1]
   }
   n9 = "Transform (Modify) Bone - Bone: dyn_belt_attach_base"(ComponentPose=@n8, Alpha=Get thigh_rot_r(), Translation="-7.5,2,0", Rotation="0,30,5", Scale="1,1,1") {
    rotationMode=BMM_Additive
    rotationSpace=BCS_BoneSpace
    translationMode=BMM_Additive
    translationSpace=BCS_ParentBoneSpace
    alphaBoolBlend {blendOption=Linear}
    alphaScaleBiasClamp {bMapRange=true, inRange {max=0, min=-100}, outRange {max=0, min=1}}
    boneToModify {boneName=dyn_belt_attach_base}
    rotation {roll=5, yaw=30}
    translation=<-7.5 2>
   }
   n10 = Get thigh_rot_side_r() -> (thigh_rot_side_r:double)
   n11 = "Transform (Modify) Bone - Bone: dyn_pouch_a_base"(ComponentPose=@n9, Alpha=@n10, Translation="0,0,0", Rotation="-35,0,0", Scale="1,1,1") {
    rotationMode=BMM_Additive
    rotationSpace=BCS_BoneSpace
    translationMode=BMM_Additive
    translationSpace=BCS_ParentBoneSpace
    alphaBoolBlend {blendOption=Linear}
    alphaScaleBiasClamp {bMapRange=true, inRange {max=35}}
    boneToModify {boneName=dyn_pouch_a_base}
    rotation {pitch=-35}
   }
   n12 = "Transform (Modify) Bone - Bone: dyn_pouch_b_base"(ComponentPose=@n11, Alpha=@n10, Translation="0,0,0", Rotation="-15,0,0", Scale="1,1,1") {
    rotationMode=BMM_Additive
    rotationSpace=BCS_BoneSpace
    translationMode=BMM_Additive
    translationSpace=BCS_ParentBoneSpace
    alphaBoolBlend {blendOption=Linear}
    alphaScaleBiasClamp {bMapRange=true, inRange {max=35}}
    boneToModify {boneName=dyn_pouch_b_base}
    rotation {pitch=-15}
   }
   n13 = Anim Dynamics - Bone: dyn_belt_main_base(ComponentPose=@n12, Alpha=Get beltAlpha() -> (beltAlpha:double), ExternalForce="0,0,0") {
    angularDampingOverride=1
    angularSpringConstant=1500
    bAngularSpring=true
    bGravityOverrideInSimSpace=true
    bOverrideAngularDamping=true
    gravityScale=0.35
    simulationSpace=RootRelative
    alphaBoolBlend {blendOption=Linear}
    boundBone {boneName=dyn_belt_main_base}
    componentLinearAccScale {z=0.45}
    componentLinearVelScale {z=0.45}
    physicsBodyDefinitions [{
      sphereCollisionRadius=0
      boundBone {boneName=dyn_belt_main_base}
      boxExtents=<10 20 20>
      constraintSetup {angularLimitsMax=<0 40 0>, angularLimitsMin=<0 0 0>, angularTarget=<0 0 0>}
      localJointOffset=<10 0 0>
     }]
   }
   n14 = Anim Dynamics - Bone: dyn_belt_attach_base(ComponentPose=@n13, Alpha=0.6, ExternalForce="0,0,0") {
    angularDampingOverride=1
    angularSpringConstant=1500
    bAngularSpring=true
    bGravityOverrideInSimSpace=true
    bOverrideAngularDamping=true
    gravityScale=0.35
    simulationSpace=RootRelative
    alphaBoolBlend {blendOption=Linear}
    boundBone {boneName=dyn_belt_attach_base}
    componentLinearAccScale=<0.45 0.45 0.45>
    componentLinearVelScale=<0.45 0.45 0.45>
    physicsBodyDefinitions [{
      sphereCollisionRadius=0
      boundBone {boneName=dyn_belt_attach_base}
      boxExtents=<10 20 20>
      constraintSetup {angularLimitsMax=<5 10 0>, angularLimitsMin=<-5 0 0>, angularTarget=<0 0 0>}
      localJointOffset=<10 0 0>
     }]
   }
   n15 = RigidBody(ComponentPose=@n14) {
    bClampLinearTranslationLimitToRefPose=true
    bOverrideWorldGravity=true
    overridePhysicsAsset=&/F_MED_Candy_Summer_Physics
    simulationSpace=BaseBoneSpace
    alphaBoolBlend {blendOption=Linear}
    baseBoneRef {boneName=pelvis}
    componentLinearAccScale=<0.3 0.3 0.3>
    componentLinearVelScale=<0.3 0.3 0.3>
    overrideWorldGravity {z=-980}
   }
   Output Pose(Result=Component To Local(ComponentPose=@n15)) {layerGroup=None}
  ]}
 EventGraph {nodes [
   Event BlueprintUpdateAnimation()
   > n1 = Sequence!(execute) -> (then_0, then_1, then_2:Exec, then_3:Exec)
   n2 = Branch(execute=@n1.then_2, Condition=Get bIsCrouching())
   > Set beltAlpha(beltAlpha:double=0) -> (Output_Get:double)
   n3 = Branch(execute=@n2.else, Condition=OR Boolean(A=Get bIsJumping(), B=Get bIsFalling()))
   > Set beltAlpha(beltAlpha=0.45)
   Set beltAlpha(execute=@n3.else, beltAlpha=0.2)
   n4 = GetOwnerSkeletalMeshComponent()
   n5 = BreakRotator(InRot=BreakTransform(InTransform=GetDeltaTransformFromRefPose(self=@n4, BoneName=thigh_r, BaseName=pelvis)).Rotation)
   Set thigh_rot_r(execute=@n1.then_1, thigh_rot_r=@n5.Yaw)
   > Set thigh_rot_side_r(thigh_rot_side_r:double=@n5.Pitch) -> (Output_Get:double)
   n6 = GetDeltaTransformFromRefPose(self=GetOwnerSkeletalMeshComponent(), BoneName=thigh_l, BaseName=pelvis)
   Set thigh_rot_l(execute=@n1.then_0, thigh_rot_l=BreakRotator(InRot=BreakTransform(InTransform=@n6).Rotation).Yaw)
  ]}
]
variables [
 beltAlpha {type=number}
 thigh_rot_l {type=number}
 thigh_rot_r {type=number}
 thigh_rot_side_r {type=number}
]
