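# Control Rig for the F_MED_Candy_Summer female character part that shifts the glasses bone when the cheeks rise.
# Reads each inner cheek bone's height relative to faceAttach and remaps it to a clamped weight.
# The larger of the two weights scales an additive local offset and yaw-rotator tilt on dyn_glasses, so either cheek alone can drive the glasses.
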ControlRigBlueprint /BRCosmetics/Characters/Player/Female/Medium/Bodies/F_MED_Candy_Summer/Meshes/Parts/F_MED_Candy_Summer_ControlRig
graphs [
 RigVMFunctionLibrary {links [], nodes []}
 RigVMModel {nodes [
   RigUnit_BeginExecution = &RigUnit_BeginExecution()
   n1 = &RigUnit_GetRelativeBoneTransform(Bone=L_cheek_inner, Space=faceAttach)
   n2 = &RigUnit_GetRelativeBoneTransform(Bone=R_cheek_inner, Space=faceAttach)
   /Script/ControlRig.RigUnit_ModifyBoneTransforms(
    BoneToModify=((Bone="dyn_glasses",Transform=(Rotation=(X=0,Y=0,Z=0,W=1),Translation=(X=-0.2,Y=0.5,Z=0),Scale3D=(X=1,Y=1,Z=1))))
    Mode=AdditiveLocal
    WeightMaximum=1
    WeightMinimum=0
    Weight=/Script/RigVM.RigVMFunction_MathFloatMax(
     A=&RigVMFunction_MathFloatRemap(SourceMaximum=5, SourceMinimum=3.86, TargetMinimum=0, bClamp=true, Value=@n1.Transform.Translation.Z).Result
     B=&RigVMFunction_MathFloatRemap(SourceMaximum=5, SourceMinimum=3.86, TargetMinimum=0, bClamp=true, Value=@n2.Transform.Translation.Z).Result
    ) {is=struct RigVMUnitNode, template="Maximum::Execute(in A,in B,out Result)"}.Result
    BoneToModify.0.Transform.Rotation=/Script/RigVM.RigVMFunction_MathQuaternionFromRotator(Rotator=(Pitch=0,Yaw=5,Roll=0)) {is=struct RigVMUnitNode}.Result
    ExecutePin=@RigUnit_BeginExecution.ExecutePin
   ) {is=struct RigVMUnitNode}
  ]}
]
hierarchy [
 root {loc=<0 0 0>, rot=<0 0 0 1>}
 pelvis {parent=root, loc=<0 -0.5301 98.0571>, rot=<0.0127 0.707 -0.0127 -0.707>}
 spine_01 {parent=pelvis, loc=<6.1208 -0.9328 0>, rot=<0 0 0.034 -0.9994>}
 spine_02 {parent=spine_01, loc=<6.5996 -0.2095 0>, rot=<0 0 0.0238 -0.9997>}
 spine_03 {parent=spine_02, loc=<6.4914 -0.1237 0>, rot=<0 0 0.0149 -0.9999>}
 spine_04 {parent=spine_03, loc=<6.9598 0.0794 0>, rot=<0 0 -0.0132 -0.9999>}
 spine_05 {parent=spine_04, loc=<6.9987 0 0>, rot=<0 0 -0.0015 -1>}
 clavicle_l {parent=spine_05, loc=<5.4831 -0.1053 -3.2946>, rot=<0.0465 -0.7289 -0.009 -0.6829>}
 upperarm_l {parent=clavicle_l, loc=<9.8658 -1.2136 -0.4104>, rot=<0.0164 -0.3245 -0.0269 -0.9454>}
 clavicle_r {parent=spine_05, loc=<5.4831 -0.1053 3.2946>, rot=<0.7289 0.0465 -0.6829 0.009>}
 upperarm_r {parent=clavicle_r, loc=<-9.8658 1.2136 0.4104>, rot=<0.0164 -0.3245 -0.0269 -0.9454>}
 neck_01 {parent=spine_05, loc=<11.3138 0.1529 0>, rot=<0 0 -0.2003 -0.9797>}
 neck_02 {parent=neck_01, loc=<4.5627 0 0>, rot=<0 0 0.0504 -0.9987>}
 head {parent=neck_02, loc=<5.3914 0 0>, rot=<0 0 0.115 -0.9934>}
 dyn_hair_ponyTail_a_base_l {parent=head, loc=<8.73 -6.7862 -6.5679>, rot=<0.0946 -0.9846 -0.0099 -0.1468>}
 dyn_hair_ponyTail_a_l {parent=dyn_hair_ponyTail_a_base_l, loc=<0 0 0>, rot=<0 0 0 -1>}
 dyn_hair_ponyTail_b_l {parent=dyn_hair_ponyTail_a_l, loc=<7.2152 0 0>, rot=<0 0 0 -1>}
 dyn_hair_ponyTail_c_l {parent=dyn_hair_ponyTail_b_l, loc=<7.2152 0 0>, rot=<0 0 0 -1>}
 dyn_hair_ponyTail_d_l {parent=dyn_hair_ponyTail_c_l, loc=<7.2152 0 0>, rot=<-0.4169 0 0 -0.909>}
 dyn_hair_ponyTail_e_l {parent=dyn_hair_ponyTail_d_l, loc=<2.1985 2.6867 -0.4878>, rot=<0.7272 0 0 -0.6864>}
 dyn_hair_ponyTail_tip_a_l {parent=dyn_hair_ponyTail_c_l, loc=<4.4263 3.3405 1.8473>, rot=<0.4854 -0.0453 -0.1677 -0.8569>}
 dyn_hair_ponyTail_tip_b_l {parent=dyn_hair_ponyTail_c_l, loc=<4.3701 -2.6982 -2.3554>, rot=<0.7644 0.0189 -0.1727 -0.6209>}
 dyn_hair_ponyTail_a_base_r {parent=head, loc=<8.73 -6.7862 6.5679>, rot=<0.9765 0.0734 -0.1584 -0.1264>}
 dyn_hair_ponyTail_a_r {parent=dyn_hair_ponyTail_a_base_r, loc=<0 0 0>, rot=<0 0 0 -1>}
 dyn_hair_ponyTail_b_r {parent=dyn_hair_ponyTail_a_r, loc=<-7.2152 0 0>, rot=<0 0 0 -1>}
 dyn_hair_ponyTail_c_r {parent=dyn_hair_ponyTail_b_r, loc=<-7.2152 0 0>, rot=<0 0 0 -1>}
 dyn_hair_ponyTail_d_r {parent=dyn_hair_ponyTail_c_r, loc=<-5.0837 0 1.1745>, rot=<-0.2835 0 0 -0.959>}
 dyn_hair_ponyTail_e_r {parent=dyn_hair_ponyTail_d_r, loc=<-2.5367 -2.2902 -0.0614>, rot=<0.7261 0.0296 -0.0225 -0.6866>}
 dyn_bun_l {parent=head, loc=<12.3816 -4.3134 -5.874>, rot=<0 -1 0 0>}
 dyn_bun_r {parent=head, loc=<12.3819 -4.3134 5.874>, rot=<0 -1 0 0>}
 dyn_hair_side_a_l {parent=head, loc=<7.0082 4.2261 -7.2728>, rot=<0.0035 -0.8386 0.5447 -0.0023>}
 dyn_hair_side_b_l {parent=head, loc=<6.6584 1.6892 -7.738>, rot=<0.0033 -0.799 0.6013 -0.0025>}
 dyn_hair_side_c_l {parent=head, loc=<6.8653 -0.5597 -8.0929>, rot=<0.003 -0.7252 0.6885 -0.0028>}
 dyn_hair_side_d_l {parent=head, loc=<9.0142 -2.1367 -7.5196>, rot=<0.003 -0.7252 0.6885 -0.0028>}
 dyn_hair_side_a_r {parent=head, loc=<7.0081 4.2261 7.2728>, rot=<0.8386 0.0035 -0.0023 -0.5447>}
 dyn_hair_side_b_r {parent=head, loc=<6.658 1.6892 7.738>, rot=<0.799 0.0033 -0.0025 -0.6013>}
 dyn_hair_side_c_r {parent=head, loc=<6.8656 -0.5597 8.0929>, rot=<0.7252 0.003 -0.0028 -0.6885>}
 dyn_hair_side_d_r {parent=head, loc=<9.0137 -2.1367 7.5196>, rot=<0.7252 0.003 -0.0028 -0.6885>}
 dyn_hair_front_l {parent=head, loc=<9.9111 7.9516 -5.2579>, rot=<0.0038 -0.9136 0.4067 -0.0017>}
 dyn_hair_front_r {parent=head, loc=<8.5522 7.851 5.3447>, rot=<0.0038 -0.9145 -0.4045 0.0017>}
 dyn_hair_front {parent=head, loc=<9.8209 10.448 0.4827>, rot=<0.0041 -0.9977 -0.067 0.0003>}
 dyn_glasses {parent=head, loc=<4.9178 -0.1825 0>, rot=<0.0041 -1 0 0>}
 faceAttach {parent=head, loc=<-3.6496 -0.5494 0.0202>, rot=<0.0082 -0.7071 -0.0082 -0.7071>, scale=<0.89 0.89 0.89>}
 C_brow_mid {parent=faceAttach, loc=<0.0068 10.2049 9.0383>, rot=<0 -0.0488 0.0096 -0.9988>}
 L_brow_mid {parent=faceAttach, loc=<3.6265 9.6831 10.6622>, rot=<0 -0.0033 0.2553 -0.9669>}
 L_brow_outer {parent=faceAttach, loc=<6.1069 7.159 10.6483>, rot=<0.1066 0.1303 0.4111 -0.8959>}
 L_cheek_inner {parent=faceAttach, loc=<4.5704 9.1326 3.7838>, rot=<-0.0181 -0.1365 0.3335 -0.9327>}
 C_lip_upper_mid {parent=faceAttach, loc=<0.0168 10.1349 1.2403>, rot=<0.2225 0.0032 -0.0012 -0.9749>}
 L_lip_upper_outer {parent=faceAttach, loc=<1.3918 9.7409 1.2136>, rot=<0.1055 0.1274 0.1813 -0.9694>}
 L_lip_corner {parent=faceAttach, loc=<2.4959 9.1736 1.2369>, rot=<0.0015 0.0095 -0.0145 -0.9998>}
 C_nose_bridge {parent=faceAttach, loc=<-0.0332 11.9408 5.3522>, rot=<-0.2218 -0.0073 -0.0021 -0.9751>}
 R_brow_mid {parent=faceAttach, loc=<-3.58 9.6834 10.6676>, rot=<-0.0001 0.0042 -0.2553 -0.9669>}
 R_brow_outer {parent=faceAttach, loc=<-6.0666 7.1519 10.6277>, rot=<0.0565 -0.0672 -0.4277 -0.8996>}
 R_cheek_inner {parent=faceAttach, loc=<-4.5125 9.1401 3.7751>, rot=<-0.0202 0.1419 -0.333 -0.932>}
 R_lip_corner {parent=faceAttach, loc=<-2.4825 9.1358 1.2388>, rot=<0.0025 -0.0085 0.0146 -0.9999>}
 C_jaw {parent=faceAttach, loc=<0.0192 2.9793 1.3786>, rot=<0 0 0 1>}
 L_lip_lower_outer {parent=C_jaw, loc=<1.444 6.5564 -0.1728>, rot=<-0.0067 -0.0618 0.2226 -0.9729>}
 C_lip_lower_mid {parent=C_jaw, loc=<0.0016 7.4433 -0.1681>, rot=<0 -0.0051 -0.0011 -1>}
 R_lip_lower_outer {parent=C_jaw, loc=<-1.4581 6.7328 -0.2319>, rot=<-0.0168 0.1169 -0.1776 -0.977>}
 teeth_lower {parent=C_jaw, loc=<-0.0701 5.3303 -1.6865>, rot=<0 0 0 1>}
 tongue {parent=C_jaw, loc=<0.0025 1.6226 -0.8121>, rot=<-0.1568 0 0 -0.9876>}
 R_eye {parent=faceAttach, loc=<-3.393 7.3361 8.0269>, rot=<0 0 0.9998 -0.0187>}
 R_eye_lid_upper_mid {parent=R_eye, loc=<-0.0396 -1.1714 0.2631>, rot=<0.0523 -0.0009 0.0167 -0.9985>}
 R_eye_lid_lower_mid {parent=R_eye, loc=<-0.0085 -0.9425 -0.7051>, rot=<0.0329 -0.0005 0.0167 -0.9993>}
 L_eye {parent=faceAttach, loc=<3.4385 7.3361 8.0269>, rot=<0 0 0.9998 0.0187>}
 L_eye_lid_upper_mid {parent=L_eye, loc=<0.0251 -1.1623 0.2388>, rot=<0.0523 0 0 -0.9986>}
 L_eye_lid_lower_mid {parent=L_eye, loc=<0.0223 -0.923 -0.7488>, rot=<0.0329 0 0 -0.9995>}
 teeth_upper {parent=faceAttach, loc=<0.0399 9.4123 2.1484>, rot=<0 0 0 1>}
 R_lip_upper_outer {parent=faceAttach, loc=<-1.3597 9.8007 1.234>, rot=<0.117 -0.1616 -0.1661 -0.9657>}
]
