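# Physics asset for the F_MED_Candy_Summer female character body part, covering the upper body and dynamic hair bones.
# Capsules on spine, neck, head, clavicles and upper arms are linked by free-rotation constraints.
# Bangs (boxes), side strands and buns hang from the head with limited swing and twist.
# Two ponytail capsule chains start at free-rotating bases, lock one swing axis at the first link, and the left adds two tip bodies.
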PhysicsAsset /BRCosmetics/Characters/Player/Female/Medium/Bodies/F_MED_Candy_Summer/Meshes/Parts/F_MED_Candy_Summer_Physics
bodies {
 clavicle_l [{shapeType=Capsule, length=5.5, radius=4.75, center=<5.9168 -1.0768 -0.3545>, rotation=<rot 79.8875 0 0>}]
 clavicle_r [{shapeType=Capsule, length=5.5, radius=4.75, center=<-5.9168 1.0768 0.3545>, rotation=<rot 79.8875 0 -180>}]
 dyn_bun_l [{shapeType=Capsule, length=2, radius=2.5, center=<2.8063 0 0>, rotation=<rot 90 0 0>}]
 dyn_bun_r [{shapeType=Capsule, length=2, radius=2.5, center=<2.8063 0 0>, rotation=<rot 90 0 180>}]
 dyn_hair_front [{shapeType=Box, extentX=4.5, extentY=1, extentZ=6, center=<2 0 0>}]
 dyn_hair_front_l [{shapeType=Box, extentX=4, extentY=1, extentZ=4.5, center=<2.2387 0.0001 0>}]
 dyn_hair_front_r [{shapeType=Box, extentX=3.5, extentY=1.5, extentZ=4.5, center=<1.42 0 0>}]
 dyn_hair_ponyTail_a_base_l [{shapeType=Capsule, length=0.505, radius=0.505, rotation=<rot 0 -90 0>}]
 dyn_hair_ponyTail_a_base_r [{shapeType=Capsule, length=0.505, radius=0.505, rotation=<rot 0 -90 180>}]
 dyn_hair_ponyTail_a_l [{shapeType=Capsule, length=2.5, radius=3, center=<3.0739 0 0>, rotation=<rot 90 0 0>}]
 dyn_hair_ponyTail_a_r [{shapeType=Capsule, length=0.75, radius=3.25, center=<-3.5485 0 0>, rotation=<rot 90 0 -179.9999>}]
 dyn_hair_ponyTail_b_l [{shapeType=Capsule, length=0.25, radius=4.75, center=<4.2996 0.0001 0.3876>, rotation=<rot 90 0 0>}]
 dyn_hair_ponyTail_b_r [{shapeType=Capsule, length=0.25, radius=3.75, center=<-3.4384 -0.0001 0>, rotation=<rot 90 0 -179.9999>}]
 dyn_hair_ponyTail_c_l [{shapeType=Capsule, length=0.75, radius=4, center=<2.8329 0 -0.1452>, rotation=<rot 90 0 0>}]
 dyn_hair_ponyTail_c_r [{shapeType=Capsule, length=0.5, radius=3.75, center=<-2 0.0001 0.6201>, rotation=<rot 90 0 -179.9999>}]
 dyn_hair_ponyTail_d_l [{shapeType=Capsule, length=1.25, radius=2, center=<1.3677 1.013 -0.3231>, rotation=<rot 90 0 -52.9466>}]
 dyn_hair_ponyTail_d_r [{shapeType=Capsule, length=1, radius=2, center=<-1.5772 -0.3601 -0.084>, rotation=<rot 90 90 46.5563>}]
 dyn_hair_ponyTail_e_l [{shapeType=Capsule, length=3.5, radius=1.25, center=<0 0 2.6863>}]
 dyn_hair_ponyTail_e_r [{shapeType=Capsule, length=4.25, radius=1.25, center=<0 0.0001 -3.0834>}]
 dyn_hair_ponyTail_tip_a_l [{shapeType=Capsule, length=1.5, radius=1, center=<0.4379 -1.5111 0>, rotation=<rot 90 -165.9637 -83.29>}]
 dyn_hair_ponyTail_tip_b_l [{shapeType=Capsule, length=1.75, radius=0.75, center=<0.9988 0.9477 0.2516>, rotation=<rot 74.0582 -142.8149 168.3289>}]
 dyn_hair_side_a_l [{shapeType=Capsule, length=1, radius=1, center=<1.5 0 0>, rotation=<rot 90 0 0>}]
 dyn_hair_side_a_r [{shapeType=Capsule, length=1, radius=1, center=<-1.5 0 0>, rotation=<rot 90 0 0>}]
 dyn_hair_side_b_l [{shapeType=Capsule, length=1, radius=1, center=<1.5 0 0>, rotation=<rot 90 0 0>}]
 dyn_hair_side_b_r [{shapeType=Capsule, length=1, radius=1, center=<-1.5 0 0>, rotation=<rot 90 0 0>}]
 dyn_hair_side_c_l [{shapeType=Capsule, length=1, radius=1, center=<1.5 0 0>, rotation=<rot 90 0 0>}]
 dyn_hair_side_c_r [{shapeType=Capsule, length=1, radius=1, center=<-1.5 0 0>, rotation=<rot 90 0 0>}]
 dyn_hair_side_d_l [{shapeType=Capsule, length=1, radius=1, center=<1.1867 0 0>, rotation=<rot 90 0 0>}]
 dyn_hair_side_d_r [{shapeType=Capsule, length=1, radius=1, center=<-1.5 0 0>, rotation=<rot 90 0 0>}]
 head [{shapeType=Capsule, length=4, radius=6.75, center=<3.6461 2.045 -0.0001>, rotation=<rot 90 -36.8699 -41.0292>}]
 neck_01 [{shapeType=Capsule, length=0.75, radius=4.5, center=<0.9973 0.3057 0>, rotation=<rot 90 0 0>}]
 neck_02 [{shapeType=Capsule, length=0.75, radius=4.25, rotation=<rot 90 0 0>}]
 spine_03 [{shapeType=Capsule, length=4.25, radius=8.25, center=<0 2.4937 0>}]
 spine_04 [{shapeType=Capsule, length=3.5, radius=10, center=<0 2.5821 0>}]
 spine_05 [{shapeType=Capsule, length=8.505, radius=8.255, center=<2.5917 0 0>, rotation=<rot 0 -90 0>}]
 upperarm_l [{shapeType=Capsule, length=20.5, radius=4, center=<11.5821 0.0002 0>, rotation=<rot 90 0 0>}]
 upperarm_r [{shapeType=Capsule, length=20.5, radius=4, center=<-11.5821 -0.0002 0>, rotation=<rot 90 0 180>}]
}
constraints [
 {
  bone1Name=upperarm_l
  bone2Name=clavicle_l
  swing1LimitDegrees=45
  swing1Motion=Free
  swing2LimitDegrees=45
  swing2Motion=Free
  twistLimitDegrees=45
  twistMotion=Free
 }
 {
  bone1Name=clavicle_l
  bone2Name=spine_05
  swing1LimitDegrees=45
  swing1Motion=Free
  swing2LimitDegrees=45
  swing2Motion=Free
  twistLimitDegrees=45
  twistMotion=Free
 }
 {
  bone1Name=head
  bone2Name=neck_02
  swing1LimitDegrees=45
  swing1Motion=Free
  swing2LimitDegrees=45
  swing2Motion=Free
  twistLimitDegrees=45
  twistMotion=Free
 }
 {
  bone1Name=neck_01
  bone2Name=spine_05
  swing1LimitDegrees=45
  swing1Motion=Free
  swing2LimitDegrees=45
  swing2Motion=Free
  twistLimitDegrees=45
  twistMotion=Free
 }
 {
  bone1Name=neck_02
  bone2Name=neck_01
  swing1LimitDegrees=45
  swing1Motion=Free
  swing2LimitDegrees=45
  swing2Motion=Free
  twistLimitDegrees=45
  twistMotion=Free
 }
 {
  bone1Name=clavicle_r
  bone2Name=spine_05
  swing1LimitDegrees=45
  swing1Motion=Free
  swing2LimitDegrees=45
  swing2Motion=Free
  twistLimitDegrees=45
  twistMotion=Free
 }
 {
  bone1Name=upperarm_r
  bone2Name=clavicle_r
  swing1LimitDegrees=45
  swing1Motion=Free
  swing2LimitDegrees=45
  swing2Motion=Free
  twistLimitDegrees=45
  twistMotion=Free
 }
 {
  bone1Name=dyn_hair_ponyTail_a_base_l
  bone2Name=head
  swing1LimitDegrees=45
  swing1Motion=Free
  swing2LimitDegrees=45
  swing2Motion=Free
  twistLimitDegrees=45
  twistMotion=Free
 }
 {
  bone1Name=dyn_hair_ponyTail_a_l
  bone2Name=dyn_hair_ponyTail_a_base_l
  swing1LimitDegrees=10
  swing2LimitDegrees=10
  swing2Motion=Locked
  twistLimitDegrees=5
 }
 {bone1Name=dyn_hair_ponyTail_b_l, bone2Name=dyn_hair_ponyTail_a_l, swing1LimitDegrees=15, swing2LimitDegrees=45, twistLimitDegrees=5}
 {bone1Name=dyn_hair_ponyTail_c_l, bone2Name=dyn_hair_ponyTail_b_l, swing1LimitDegrees=30, swing2LimitDegrees=45, twistLimitDegrees=5}
 {bone1Name=dyn_hair_ponyTail_d_l, bone2Name=dyn_hair_ponyTail_c_l, swing1LimitDegrees=10, swing2LimitDegrees=5, twistLimitDegrees=15}
 {bone1Name=dyn_hair_ponyTail_e_l, bone2Name=dyn_hair_ponyTail_d_l, swing1LimitDegrees=5, swing2LimitDegrees=10, twistLimitDegrees=15}
 {
  bone1Name=dyn_hair_ponyTail_a_r
  bone2Name=dyn_hair_ponyTail_a_base_r
  swing1LimitDegrees=10
  swing2LimitDegrees=10
  swing2Motion=Locked
  twistLimitDegrees=5
 }
 {bone1Name=dyn_hair_ponyTail_b_r, bone2Name=dyn_hair_ponyTail_a_r, swing1LimitDegrees=30, swing2LimitDegrees=45, twistLimitDegrees=5}
 {bone1Name=dyn_hair_ponyTail_c_r, bone2Name=dyn_hair_ponyTail_b_r, swing1LimitDegrees=30, swing2LimitDegrees=45, twistLimitDegrees=5}
 {bone1Name=dyn_hair_ponyTail_d_r, bone2Name=dyn_hair_ponyTail_c_r, swing1LimitDegrees=5, swing2LimitDegrees=5, twistLimitDegrees=15}
 {bone1Name=dyn_hair_ponyTail_e_r, bone2Name=dyn_hair_ponyTail_d_r, swing1LimitDegrees=5, swing2LimitDegrees=10, twistLimitDegrees=15}
 {
  bone1Name=dyn_hair_ponyTail_a_base_r
  bone2Name=head
  swing1LimitDegrees=45
  swing1Motion=Free
  swing2LimitDegrees=45
  swing2Motion=Free
  twistLimitDegrees=45
  twistMotion=Free
 }
 {
  bone1Name=spine_05
  bone2Name=spine_04
  swing1LimitDegrees=45
  swing1Motion=Free
  swing2LimitDegrees=45
  swing2Motion=Free
  twistLimitDegrees=45
  twistMotion=Free
 }
 {
  bone1Name=spine_04
  bone2Name=spine_03
  swing1LimitDegrees=45
  swing1Motion=Free
  swing2LimitDegrees=45
  swing2Motion=Free
  twistLimitDegrees=45
  twistMotion=Free
 }
 {bone1Name=dyn_hair_ponyTail_tip_a_l, bone2Name=dyn_hair_ponyTail_d_l, swing1LimitDegrees=5, swing2LimitDegrees=5, twistLimitDegrees=15}
 {bone1Name=dyn_hair_ponyTail_tip_b_l, bone2Name=dyn_hair_ponyTail_d_l, swing1LimitDegrees=5, swing2LimitDegrees=5, twistLimitDegrees=15}
 {bone1Name=dyn_hair_front_l, bone2Name=head, swing1LimitDegrees=45, swing2LimitDegrees=10, twistLimitDegrees=5}
 {bone1Name=dyn_hair_front_r, bone2Name=head, swing1LimitDegrees=45, swing2LimitDegrees=5, twistLimitDegrees=5}
 {bone1Name=dyn_hair_front, bone2Name=head, swing1LimitDegrees=45, swing2LimitDegrees=10, twistLimitDegrees=5}
 {bone1Name=dyn_hair_side_a_l, bone2Name=head, swing1LimitDegrees=15, swing2LimitDegrees=15, twistLimitDegrees=5}
 {bone1Name=dyn_hair_side_b_l, bone2Name=head, swing1LimitDegrees=15, swing2LimitDegrees=15, twistLimitDegrees=5}
 {bone1Name=dyn_hair_side_c_l, bone2Name=head, swing1LimitDegrees=25, swing2LimitDegrees=25, twistLimitDegrees=5}
 {bone1Name=dyn_hair_side_d_l, bone2Name=head, swing1LimitDegrees=15, swing2LimitDegrees=15, twistLimitDegrees=5}
 {bone1Name=dyn_hair_side_a_r, bone2Name=head, swing1LimitDegrees=15, swing2LimitDegrees=15, twistLimitDegrees=5}
 {bone1Name=dyn_hair_side_b_r, bone2Name=head, swing1LimitDegrees=15, swing2LimitDegrees=15, twistLimitDegrees=5}
 {bone1Name=dyn_hair_side_c_r, bone2Name=head, swing1LimitDegrees=25, swing2LimitDegrees=25, twistLimitDegrees=5}
 {bone1Name=dyn_hair_side_d_r, bone2Name=head, swing1LimitDegrees=15, swing2LimitDegrees=15, twistLimitDegrees=5}
 {bone1Name=dyn_bun_l, bone2Name=head, swing1LimitDegrees=5, swing2LimitDegrees=15, twistLimitDegrees=15}
 {bone1Name=dyn_bun_r, bone2Name=head, swing1LimitDegrees=5, swing2LimitDegrees=15, twistLimitDegrees=15}
]
